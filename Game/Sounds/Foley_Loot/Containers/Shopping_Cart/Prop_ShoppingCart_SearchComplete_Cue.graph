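# Sound cue for the foley played when searching a shopping cart loot container completes, assigned to the loot collect sound class.
# A Random node picks one of two equally weighted shopping cart search waves, which passes through a Modulator into an Enveloper that fades volume out over a short window.
# Prop search attenuation shapes falloff, and a concurrency override allows only one instance, preventing new plays while one is active.
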SoundCue /Game/Sounds/Foley_Loot/Containers/Shopping_Cart/Prop_ShoppingCart_SearchComplete_Cue
root=Enveloper_1
nodes [
 {
  id=Enveloper_1
  children [Modulator_3]
  properties {pitchCurve {editorCurveData {keys [{time=0}]}}, volumeCurve {editorCurveData {keys [<at 0 1>,<at 0.7 0>]}}}
 }
 {id=Modulator_3, children [Random_5]}
 {id=Random_5, children [WavePlayer_12, WavePlayer_13], properties {weights [1,1]}}
 {id=WavePlayer_12, properties {soundWaveAssetPtr=&/shoppingcart_search_01}}
 {id=WavePlayer_13, properties {soundWaveAssetPtr=&/shoppingcart_search_02}}
]
properties {
 attenuationSettings=/Game/Sounds/Attenuation/Fort_Prop_Search_Att
 bOverrideConcurrency=true
 maxDistance=1300
 soundClassObject=&Fort_Loot_Collect
 volumeMultiplier=0.8
 concurrencyOverrides {maxCount=1, resolutionRule=PreventNew}
}
