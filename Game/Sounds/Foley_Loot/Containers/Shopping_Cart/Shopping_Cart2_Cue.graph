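# Sound cue for shopping cart loot container foley, in the Fort_Loot_Collect sound class.
# A single wave player feeds the Shopping_Cart2 sound straight to the output at reduced volume, with no randomization or modulation.
# Inline attenuation overrides take the place of the referenced Fort_Prop_Search_Att asset to set a local falloff range.
# Inline concurrency overrides block new instances once the limit is reached.
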SoundCue /Game/Sounds/Foley_Loot/Containers/Shopping_Cart/Shopping_Cart2_Cue
root=WavePlayer_1
nodes [{id=WavePlayer_1, properties {soundWaveAssetPtr=&/Shopping_Cart2}}]
properties {
 attenuationSettings=/Game/Sounds/Attenuation/Fort_Prop_Search_Att
 bOverrideAttenuation=true
 bOverrideConcurrency=true
 maxDistance=1500
 soundClassObject=&Fort_Loot_Collect
 volumeMultiplier=0.3
 attenuationOverrides {falloffDistance=1000, attenuationShapeExtents {x=500}}
 concurrencyOverrides {resolutionRule=PreventNew}
}
